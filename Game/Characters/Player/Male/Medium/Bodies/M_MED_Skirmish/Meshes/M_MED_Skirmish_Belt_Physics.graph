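# Physics asset for the Skirmish belt of the medium male player body, giving collision and joints to its front loincloth and hanging straps.
# Capsule bodies cover the pelvis, center, left and right cloth flaps, and a three-segment strap chain per side.
# Tiny base bodies link each flap and strap to the pelvis with free swing and twist.
# Flaps also rotate freely, while straps use tight angular limits, with twist locked on the lower segments.
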PhysicsAsset /Game/Characters/Player/Male/Medium/Bodies/M_MED_Skirmish/Meshes/M_MED_Skirmish_Belt_Physics
bodies {
 dyn_base_loinCloth_fr_base_l [{shapeType=Capsule, length=0.505, radius=0.505, rotation=<rot 0 -90 0>}]
 dyn_base_roinCloth_fr_base_r [{shapeType=Capsule, length=0.505, radius=0.505, rotation=<rot 0 -90 0>}]
 dyn_loinCloth_fr [{shapeType=Capsule, length=14, radius=13.5, center=<9.1945 -13.464 0>, rotation=<rot 90 0 0>}]
 dyn_loinCloth_fr_base [{shapeType=Capsule, length=0.505, radius=0.505, rotation=<rot 0 -90 0>}]
 dyn_loinCloth_fr_l [{shapeType=Capsule, length=16, radius=11, center=<11.3768 -9.6208 -4.2623>, rotation=<rot 90 0 0>}]
 dyn_roinCloth_fr_r [{shapeType=Capsule, length=14.75, radius=12.25, center=<-12.5747 11.4905 3.7403>, rotation=<rot 90 0 0>}]
 dyn_strap_a_base_l [{shapeType=Capsule, length=0.505, radius=0.505, rotation=<rot 0 -90 0>}]
 dyn_strap_a_base_r [{shapeType=Capsule, length=0.505, radius=0.505, rotation=<rot 0 -90 0>}]
 dyn_strap_a_l [{shapeType=Capsule, length=2.25, radius=1.75, center=<3.107 -0.0001 0>, rotation=<rot 90 0 0>}]
 dyn_strap_a_r [{shapeType=Capsule, length=3, radius=2, center=<-3.0856 0 0>, rotation=<rot 90 0 0>}]
 dyn_strap_b_l [{shapeType=Capsule, length=3, radius=2.25, center=<2.9998 0 0>, rotation=<rot 90 0 0>}]
 dyn_strap_b_r [{shapeType=Capsule, length=2.75, radius=2.5, center=<-2.9443 0 0>, rotation=<rot 90 0 0>}]
 dyn_strap_c_l [{shapeType=Capsule, length=5, radius=2.25, center=<3.8646 0 0>, rotation=<rot 90 0 0>}]
 dyn_strap_c_r [{shapeType=Capsule, length=4.25, radius=2.5, center=<-3.7882 0 0>, rotation=<rot 90 0 0>}]
 pelvis [{shapeType=Capsule, length=13.5739, radius=12.1217, center=<-6.8429 -0.3954 -0.2412>}]
}
constraints [
 {bone1Name=dyn_strap_a_l, bone2Name=dyn_strap_a_base_l, swing1LimitDegrees=30, swing2LimitDegrees=10, twistLimitDegrees=5}
 {bone1Name=dyn_strap_b_l, bone2Name=dyn_strap_a_l, swing1LimitDegrees=5, swing2LimitDegrees=5, twistLimitDegrees=5, twistMotion=Locked}
 {bone1Name=dyn_strap_c_l, bone2Name=dyn_strap_b_l, swing1LimitDegrees=5, swing2LimitDegrees=5, twistLimitDegrees=5, twistMotion=Locked}
 {bone1Name=dyn_strap_a_r, bone2Name=dyn_strap_a_base_r, swing1LimitDegrees=30, swing2LimitDegrees=10, twistLimitDegrees=5}
 {bone1Name=dyn_strap_b_r, bone2Name=dyn_strap_a_r, swing1LimitDegrees=5, swing2LimitDegrees=5, twistLimitDegrees=5, twistMotion=Locked}
 {bone1Name=dyn_strap_c_r, bone2Name=dyn_strap_b_r, swing1LimitDegrees=5, swing2LimitDegrees=5, twistLimitDegrees=5, twistMotion=Locked}
 {
  bone1Name=dyn_base_loinCloth_fr_base_l
  bone2Name=pelvis
  swing1LimitDegrees=45
  swing1Motion=Free
  swing2LimitDegrees=45
  swing2Motion=Free
  twistLimitDegrees=45
  twistMotion=Free
 }
 {
  bone1Name=dyn_loinCloth_fr_base
  bone2Name=pelvis
  swing1LimitDegrees=45
  swing1Motion=Free
  swing2LimitDegrees=45
  swing2Motion=Free
  twistLimitDegrees=45
  twistMotion=Free
 }
 {
  bone1Name=dyn_base_roinCloth_fr_base_r
  bone2Name=pelvis
  swing1LimitDegrees=45
  swing1Motion=Free
  swing2LimitDegrees=45
  swing2Motion=Free
  twistLimitDegrees=45
  twistMotion=Free
 }
 {
  bone1Name=dyn_strap_a_base_r
  bone2Name=pelvis
  swing1LimitDegrees=45
  swing1Motion=Free
  swing2LimitDegrees=45
  swing2Motion=Free
  twistLimitDegrees=45
  twistMotion=Free
 }
 {
  bone1Name=dyn_strap_a_base_l
  bone2Name=pelvis
  swing1LimitDegrees=45
  swing1Motion=Free
  swing2LimitDegrees=45
  swing2Motion=Free
  twistLimitDegrees=45
  twistMotion=Free
 }
 {
  bone1Name=dyn_loinCloth_fr
  bone2Name=dyn_loinCloth_fr_base
  swing1LimitDegrees=45
  swing1Motion=Free
  swing2LimitDegrees=45
  swing2Motion=Free
  twistLimitDegrees=45
  twistMotion=Free
 }
 {
  bone1Name=dyn_loinCloth_fr_l
  bone2Name=dyn_base_loinCloth_fr_base_l
  swing1LimitDegrees=45
  swing1Motion=Free
  swing2LimitDegrees=45
  swing2Motion=Free
  twistLimitDegrees=45
  twistMotion=Free
 }
 {
  bone1Name=dyn_roinCloth_fr_r
  bone2Name=dyn_base_roinCloth_fr_base_r
  swing1LimitDegrees=45
  swing1Motion=Free
  swing2LimitDegrees=45
  swing2Motion=Free
  twistLimitDegrees=45
  twistMotion=Free
 }
]
