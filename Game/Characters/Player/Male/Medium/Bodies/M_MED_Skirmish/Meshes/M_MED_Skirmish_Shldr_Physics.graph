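# Physics asset for the shoulder mesh of the medium male Skirmish player body, setting up secondary motion on its left and right shoulder pieces.
# Each dyn_shldr bone (a to l per side) carries a capsule and is constrained to its own small-capsule _sdk bone.
# Tight swing and twist limits allow only a few degrees of movement.
# Large capsules on both upper arms have no constraints, apparently serving as colliders for the pieces.
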PhysicsAsset /Game/Characters/Player/Male/Medium/Bodies/M_MED_Skirmish/Meshes/M_MED_Skirmish_Shldr_Physics
bodies {
 dyn_shldr_a_l [{shapeType=Capsule, length=3.25, radius=1.75, center=<2.2795 0 0>, rotation=<rot 90 0 0>}]
 dyn_shldr_a_r [{shapeType=Capsule, length=3.5, radius=1.75, center=<-2.2795 0 0>, rotation=<rot 90 0 180>}]
 dyn_shldr_a_sdk_l [{shapeType=Capsule, length=0.505, radius=0.505, rotation=<rot 0 -90 0>}]
 dyn_shldr_a_sdk_r [{shapeType=Capsule, length=0.505, radius=0.505, rotation=<rot 0 -90 0>}]
 dyn_shldr_b_l [{shapeType=Capsule, length=3, radius=1.75, center=<2.2216 0.0001 0>, rotation=<rot 90 0 0>}]
 dyn_shldr_b_r [{shapeType=Capsule, length=4.25, radius=2.25, center=<-3.0813 -0.0002 0>, rotation=<rot 90 0 -179.9999>}]
 dyn_shldr_b_sdk_l [{shapeType=Capsule, length=0.505, radius=0.505, rotation=<rot 0 -90 0>}]
 dyn_shldr_b_sdk_r [{shapeType=Capsule, length=0.505, radius=0.505, rotation=<rot 0 -90 0>}]
 dyn_shldr_c_l [{shapeType=Capsule, length=2.25, radius=1.75, center=<2.8184 0 0>, rotation=<rot 90 0 0>}]
 dyn_shldr_c_r [{shapeType=Capsule, length=3.5, radius=1.75, center=<-3.2948 0.0001 0>, rotation=<rot 90 0 -179.9999>}]
 dyn_shldr_c_sdk_l [{shapeType=Capsule, length=0.505, radius=0.505, rotation=<rot 0 -90 0>}]
 dyn_shldr_c_sdk_r [{shapeType=Capsule, length=0.505, radius=0.505, rotation=<rot 0 -90 0>}]
 dyn_shldr_d_l [{shapeType=Capsule, length=3.75, radius=1.75, center=<3 0 0>, rotation=<rot 90 0 0>}]
 dyn_shldr_d_r [{shapeType=Capsule, length=5, radius=2.75, center=<-3.4651 0 0>, rotation=<rot 90 0 -179.9999>}]
 dyn_shldr_d_sdk_l [{shapeType=Capsule, length=0.505, radius=0.505, rotation=<rot 0 -90 0>}]
 dyn_shldr_d_sdk_r [{shapeType=Capsule, length=0.505, radius=0.505, rotation=<rot 0 -90 0>}]
 dyn_shldr_e_l [{shapeType=Capsule, length=3.75, radius=1.75, center=<3 0 0>, rotation=<rot 90 0 0>}]
 dyn_shldr_e_r [{shapeType=Capsule, length=3.75, radius=1.75, center=<-3 0 0>, rotation=<rot 90 0 180>}]
 dyn_shldr_e_sdk_l [{shapeType=Capsule, length=0.505, radius=0.505, rotation=<rot 0 -90 0>}]
 dyn_shldr_e_sdk_r [{shapeType=Capsule, length=0.505, radius=0.505, rotation=<rot 0 -90 0>}]
 dyn_shldr_f_l [{shapeType=Capsule, length=3.75, radius=1.75, center=<3 0 0>, rotation=<rot 90 0 0>}]
 dyn_shldr_f_r [{shapeType=Capsule, length=3.75, radius=1.75, center=<-3 0 0>, rotation=<rot 90 0 180>}]
 dyn_shldr_f_sdk_l [{shapeType=Capsule, length=0.505, radius=0.505, rotation=<rot 0 -90 0>}]
 dyn_shldr_f_sdk_r [{shapeType=Capsule, length=0.505, radius=0.505, rotation=<rot 0 -90 0>}]
 dyn_shldr_g_l [{shapeType=Capsule, length=3.75, radius=1.75, center=<3 0 0>, rotation=<rot 90 0 0>}]
 dyn_shldr_g_r [{shapeType=Capsule, length=3.75, radius=1.75, center=<-3 0 0>, rotation=<rot 90 0 180>}]
 dyn_shldr_g_sdk_l [{shapeType=Capsule, length=0.505, radius=0.505, rotation=<rot 0 -90 0>}]
 dyn_shldr_g_sdk_r [{shapeType=Capsule, length=0.505, radius=0.505, rotation=<rot 0 -90 0>}]
 dyn_shldr_h_l [{shapeType=Capsule, length=3.75, radius=1.75, center=<3 0 0>, rotation=<rot 90 0 0>}]
 dyn_shldr_h_r [{shapeType=Capsule, length=3.75, radius=1.75, center=<-3 0 0>, rotation=<rot 90 0 180>}]
 dyn_shldr_h_sdk_l [{shapeType=Capsule, length=0.505, radius=0.505, rotation=<rot 0 -90 0>}]
 dyn_shldr_h_sdk_r [{shapeType=Capsule, length=0.505, radius=0.505, rotation=<rot 0 -90 0>}]
 dyn_shldr_i_l [{shapeType=Capsule, length=2.75, radius=1.75, center=<1.4353 0 0>, rotation=<rot 90 0 0>}]
 dyn_shldr_i_r [{shapeType=Capsule, length=3.25, radius=1.75, center=<-1.4353 0 0>, rotation=<rot 90 -49.4149 130.585>}]
 dyn_shldr_i_sdk_l [{shapeType=Capsule, length=0.505, radius=0.505, rotation=<rot 0 -90 0>}]
 dyn_shldr_i_sdk_r [{shapeType=Capsule, length=0.505, radius=0.505, rotation=<rot 0 -90 0>}]
 dyn_shldr_j_l [{shapeType=Capsule, length=2.25, radius=1.5, center=<3 0 0>, rotation=<rot 90 0 0>}]
 dyn_shldr_j_r [{shapeType=Capsule, length=2.25, radius=1.5, center=<-3 0 0>, rotation=<rot 90 0 180>}]
 dyn_shldr_j_sdk_l [{shapeType=Capsule, length=0.505, radius=0.505, rotation=<rot 0 -90 0>}]
 dyn_shldr_j_sdk_r [{shapeType=Capsule, length=0.505, radius=0.505, rotation=<rot 0 -90 0>}]
 dyn_shldr_k_l [{shapeType=Capsule, length=2.25, radius=1.5, center=<3 0 0>, rotation=<rot 90 0 0>}]
 dyn_shldr_k_r [{shapeType=Capsule, length=2.25, radius=1.5, center=<-3 0 0>, rotation=<rot 90 0 180>}]
 dyn_shldr_k_sdk_l [{shapeType=Capsule, length=0.505, radius=0.505, rotation=<rot 0 -90 0>}]
 dyn_shldr_k_sdk_r [{shapeType=Capsule, length=0.505, radius=0.505, rotation=<rot 0 -90 0>}]
 dyn_shldr_l_l [{shapeType=Capsule, length=2.25, radius=1.5, center=<3 0 0>, rotation=<rot 90 0 0>}]
 dyn_shldr_l_r [{shapeType=Capsule, length=2.25, radius=1.5, center=<-3 0 0>, rotation=<rot 90 0 180>}]
 dyn_shldr_l_sdk_l [{shapeType=Capsule, length=0.505, radius=0.505, rotation=<rot 0 -90 0>}]
 dyn_shldr_l_sdk_r [{shapeType=Capsule, length=0.505, radius=0.505, rotation=<rot 0 -90 0>}]
 upperarm_l [{shapeType=Capsule, length=23, radius=6.25, center=<14.4272 -1.1766 0.6231>, rotation=<rot 85.3413 4.3579 -4.2544>}]
 upperarm_r [{shapeType=Capsule, length=23, radius=6.25, center=<-14.4272 1.1766 -0.6231>, rotation=<rot 85.3412 4.3579 175.7456>}]
}
constraints [
 {bone1Name=dyn_shldr_a_l, bone2Name=dyn_shldr_a_sdk_l, swing1LimitDegrees=5, swing2LimitDegrees=2, twistLimitDegrees=2}
 {bone1Name=dyn_shldr_b_l, bone2Name=dyn_shldr_b_sdk_l, swing1LimitDegrees=5, swing2LimitDegrees=2, twistLimitDegrees=2}
 {bone1Name=dyn_shldr_c_l, bone2Name=dyn_shldr_c_sdk_l, swing1LimitDegrees=5, swing2LimitDegrees=2, twistLimitDegrees=2}
 {bone1Name=dyn_shldr_d_l, bone2Name=dyn_shldr_d_sdk_l, swing1LimitDegrees=5, swing2LimitDegrees=2, twistLimitDegrees=2}
 {bone1Name=dyn_shldr_e_l, bone2Name=dyn_shldr_e_sdk_l, swing1LimitDegrees=5, swing2LimitDegrees=2, twistLimitDegrees=2}
 {bone1Name=dyn_shldr_f_l, bone2Name=dyn_shldr_f_sdk_l, swing1LimitDegrees=5, swing2LimitDegrees=2, twistLimitDegrees=2}
 {bone1Name=dyn_shldr_g_l, bone2Name=dyn_shldr_g_sdk_l, swing1LimitDegrees=5, swing2LimitDegrees=2, twistLimitDegrees=2}
 {bone1Name=dyn_shldr_h_l, bone2Name=dyn_shldr_h_sdk_l, swing1LimitDegrees=5, swing2LimitDegrees=2, twistLimitDegrees=2}
 {bone1Name=dyn_shldr_i_l, bone2Name=dyn_shldr_i_sdk_l, swing1LimitDegrees=5, swing2LimitDegrees=2, twistLimitDegrees=2}
 {bone1Name=dyn_shldr_a_r, bone2Name=dyn_shldr_a_sdk_r, swing1LimitDegrees=5, swing2LimitDegrees=2, twistLimitDegrees=2}
 {bone1Name=dyn_shldr_b_r, bone2Name=dyn_shldr_b_sdk_r, swing1LimitDegrees=5, swing2LimitDegrees=2, twistLimitDegrees=2}
 {bone1Name=dyn_shldr_c_r, bone2Name=dyn_shldr_c_sdk_r, swing1LimitDegrees=5, swing2LimitDegrees=2, twistLimitDegrees=2}
 {bone1Name=dyn_shldr_d_r, bone2Name=dyn_shldr_d_sdk_r, swing1LimitDegrees=5, swing2LimitDegrees=2, twistLimitDegrees=2}
 {bone1Name=dyn_shldr_e_r, bone2Name=dyn_shldr_e_sdk_r, swing1LimitDegrees=5, swing2LimitDegrees=2, twistLimitDegrees=2}
 {bone1Name=dyn_shldr_f_r, bone2Name=dyn_shldr_f_sdk_r, swing1LimitDegrees=5, swing2LimitDegrees=2, twistLimitDegrees=2}
 {bone1Name=dyn_shldr_g_r, bone2Name=dyn_shldr_g_sdk_r, swing1LimitDegrees=5, swing2LimitDegrees=2, twistLimitDegrees=2}
 {bone1Name=dyn_shldr_h_r, bone2Name=dyn_shldr_h_sdk_r, swing1LimitDegrees=5, swing2LimitDegrees=2, twistLimitDegrees=2}
 {bone1Name=dyn_shldr_i_r, bone2Name=dyn_shldr_i_sdk_r, swing1LimitDegrees=5, swing2LimitDegrees=2, twistLimitDegrees=2}
 {bone1Name=dyn_shldr_k_l, bone2Name=dyn_shldr_k_sdk_l, swing1LimitDegrees=5, swing2LimitDegrees=5, twistLimitDegrees=2}
 {bone1Name=dyn_shldr_l_l, bone2Name=dyn_shldr_l_sdk_l, swing1LimitDegrees=5, swing2LimitDegrees=2, twistLimitDegrees=2}
 {bone1Name=dyn_shldr_j_l, bone2Name=dyn_shldr_j_sdk_l, swing1LimitDegrees=5, swing2LimitDegrees=2, twistLimitDegrees=2}
 {bone1Name=dyn_shldr_j_r, bone2Name=dyn_shldr_j_sdk_r, swing1LimitDegrees=5, swing2LimitDegrees=2, twistLimitDegrees=2}
 {bone1Name=dyn_shldr_k_r, bone2Name=dyn_shldr_k_sdk_r, swing1LimitDegrees=5, swing2LimitDegrees=5, twistLimitDegrees=2}
 {bone1Name=dyn_shldr_l_r, bone2Name=dyn_shldr_l_sdk_r, swing1LimitDegrees=5, swing2LimitDegrees=2, twistLimitDegrees=2}
]
