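# Physics asset for the loincloth on the medium male Skirmish player body, with six hanging flaps (front and back, each center, left and right).
# Each flap is constrained with tight swing and twist limits to its own tiny base capsule, which links to the pelvis with free rotation.
# Front flaps use boxes and back flaps use capsules, alongside pelvis and thigh capsules for the body.
# Right-side bones are spelled "roinCloth".
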PhysicsAsset /Game/Characters/Player/Male/Medium/Bodies/M_MED_Skirmish/Meshes/M_MED_Skirmish_loinCloth_Physics
bodies {
 dyn_base_loinCloth_bk_base_l [{shapeType=Capsule, length=0.505, radius=0.505, rotation=<rot 0 -90 0>}]
 dyn_base_loinCloth_fr_base_l [{shapeType=Capsule, length=0.505, radius=0.505, rotation=<rot 0 -90 0>}]
 dyn_base_roinCloth_bk_base_r [{shapeType=Capsule, length=0.505, radius=0.505, rotation=<rot 0 -90 0>}]
 dyn_base_roinCloth_fr_base_r [{shapeType=Capsule, length=0.505, radius=0.505, rotation=<rot 0 -90 0>}]
 dyn_loinCloth_bk [{shapeType=Capsule, length=7.75, radius=3, center=<5 0 0>, rotation=<rot 90 0 0>}]
 dyn_loinCloth_bk_base [{shapeType=Capsule, length=0.505, radius=0.505, rotation=<rot 0 -90 0>}]
 dyn_loinCloth_bk_l [{shapeType=Capsule, length=7.75, radius=3, center=<5 0 0>, rotation=<rot 90 0 0>}]
 dyn_loinCloth_fr [{shapeType=Box, extentX=14, extentY=2, extentZ=18, center=<6.8572 0 0>}]
 dyn_loinCloth_fr_base [{shapeType=Capsule, length=0.505, radius=0.505, rotation=<rot 0 -90 0>}]
 dyn_loinCloth_fr_l [{shapeType=Box, extentX=14, extentY=2, extentZ=13.5, center=<6.9789 0 0>, rotation=<rot 0 0 -16.847>}]
 dyn_roinCloth_bk_r [{shapeType=Capsule, length=7.75, radius=3, center=<-5 0 0>, rotation=<rot 90 0 0>}]
 dyn_roinCloth_fr_r [{shapeType=Box, extentX=14, extentY=2, extentZ=13.5, center=<-6.9789 0 0>, rotation=<rot 0 0 -16.847>}]
 pelvis [{shapeType=Capsule, length=14.5739, radius=11.1217, center=<-6.8429 -0.3954 -0.2412>}]
 thigh_l [{shapeType=Capsule, length=20.5, radius=10, center=<-19.889 -0.0505 0>, rotation=<rot 90 0.8952 -0.0969>}]
 thigh_r [{shapeType=Capsule, length=20.5, radius=10, center=<19.889 0.0505 0>, rotation=<rot 90 -0.2238 178.7841>}]
}
constraints [
 {
  bone1Name=thigh_l
  bone2Name=pelvis
  swing1LimitDegrees=45
  swing1Motion=Free
  swing2LimitDegrees=45
  swing2Motion=Free
  twistLimitDegrees=45
  twistMotion=Free
 }
 {
  bone1Name=thigh_r
  bone2Name=pelvis
  swing1LimitDegrees=45
  swing1Motion=Free
  swing2LimitDegrees=45
  swing2Motion=Free
  twistLimitDegrees=45
  twistMotion=Free
 }
 {
  bone1Name=dyn_base_loinCloth_fr_base_l
  bone2Name=pelvis
  swing1LimitDegrees=45
  swing1Motion=Free
  swing2LimitDegrees=45
  swing2Motion=Free
  twistLimitDegrees=45
  twistMotion=Free
 }
 {bone1Name=dyn_loinCloth_fr_l, bone2Name=dyn_base_loinCloth_fr_base_l, swing1LimitDegrees=30, swing2LimitDegrees=5, twistLimitDegrees=5}
 {
  bone1Name=dyn_loinCloth_fr_base
  bone2Name=pelvis
  swing1LimitDegrees=45
  swing1Motion=Free
  swing2LimitDegrees=45
  swing2Motion=Free
  twistLimitDegrees=45
  twistMotion=Free
 }
 {bone1Name=dyn_loinCloth_fr, bone2Name=dyn_loinCloth_fr_base, swing1LimitDegrees=30, swing2LimitDegrees=10, twistLimitDegrees=10}
 {
  bone1Name=dyn_base_roinCloth_fr_base_r
  bone2Name=pelvis
  swing1LimitDegrees=45
  swing1Motion=Free
  swing2LimitDegrees=45
  swing2Motion=Free
  twistLimitDegrees=45
  twistMotion=Free
 }
 {bone1Name=dyn_roinCloth_fr_r, bone2Name=dyn_base_roinCloth_fr_base_r, swing1LimitDegrees=30, swing2LimitDegrees=5, twistLimitDegrees=5}
 {
  bone1Name=dyn_base_loinCloth_bk_base_l
  bone2Name=pelvis
  swing1LimitDegrees=45
  swing1Motion=Free
  swing2LimitDegrees=45
  swing2Motion=Free
  twistLimitDegrees=45
  twistMotion=Free
 }
 {
  bone1Name=dyn_base_roinCloth_bk_base_r
  bone2Name=pelvis
  swing1LimitDegrees=45
  swing1Motion=Free
  swing2LimitDegrees=45
  swing2Motion=Free
  twistLimitDegrees=45
  twistMotion=Free
 }
 {
  bone1Name=dyn_loinCloth_bk_base
  bone2Name=pelvis
  swing1LimitDegrees=45
  swing1Motion=Free
  swing2LimitDegrees=45
  swing2Motion=Free
  twistLimitDegrees=45
  twistMotion=Free
 }
 {bone1Name=dyn_loinCloth_bk, bone2Name=dyn_loinCloth_bk_base, swing1LimitDegrees=25, swing2LimitDegrees=5, twistLimitDegrees=5}
 {bone1Name=dyn_loinCloth_bk_l, bone2Name=dyn_base_loinCloth_bk_base_l, swing1LimitDegrees=25, swing2LimitDegrees=5, twistLimitDegrees=5}
 {bone1Name=dyn_roinCloth_bk_r, bone2Name=dyn_base_roinCloth_bk_base_r, swing1LimitDegrees=25, swing2LimitDegrees=5, twistLimitDegrees=5}
]
