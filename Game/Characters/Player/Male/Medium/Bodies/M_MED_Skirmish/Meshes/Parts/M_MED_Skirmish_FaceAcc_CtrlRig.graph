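# Control Rig for the Skirmish medium male face accessory that nudges three clip-fix bones as the brows rise, apparently to prevent clipping.
# Left, right and center brow-mid bone heights, read relative to faceAttach, are remapped to clamped weights.
# Each weight scales an additive local Y offset on its matching dyn_clip_fix bone, a child of head.
# The hierarchy also carries face, hair and strap bones, but the graph modifies only those three.
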ControlRigBlueprint /Game/Characters/Player/Male/Medium/Bodies/M_MED_Skirmish/Meshes/Parts/M_MED_Skirmish_FaceAcc_CtrlRig
graphs [
 RigVMFunctionLibrary {links [], nodes []}
 RigVMModel {nodes [
   n1 = &RigUnit_BeginExecution()
   /Script/ControlRig.RigUnit_ModifyBoneTransforms(
    BoneToModify=((Bone="dyn_clip_fix_l",Transform=(Rotation=(X=0,Y=0,Z=0,W=1),Translation=(X=0,Y=0.5,Z=0),Scale3D=(X=1,Y=1,Z=1))))
    Mode=AdditiveLocal
    WeightMaximum=1
    WeightMinimum=0
    ExecutePin=@n1.ExecutePin
    Weight=&RigVMFunction_MathFloatRemap(
     SourceMaximum=9
     SourceMinimum=8.25
     TargetMinimum=0
     bClamp=true
     Value=&RigUnit_GetRelativeBoneTransform(Bone=L_brow_mid, Space=faceAttach).Transform.Translation.Z
    ).Result
   ) {is=struct RigVMUnitNode}
   > &RigUnit_ModifyBoneTransforms(
    BoneToModify=((Bone="dyn_clip_fix_r",Transform=(Rotation=(X=0,Y=0,Z=0,W=1),Translation=(X=0,Y=0.5,Z=0),Scale3D=(X=1,Y=1,Z=1))))
    Mode=AdditiveLocal
    WeightMaximum=1
    WeightMinimum=0
    Weight=&RigVMFunction_MathFloatRemap(
     SourceMaximum=9
     SourceMinimum=8.25
     TargetMinimum=0
     bClamp=true
     Value=&RigUnit_GetRelativeBoneTransform(Bone=R_brow_mid, Space=faceAttach).Transform.Translation.Z
    ).Result
   )
   > &RigUnit_ModifyBoneTransforms(
    BoneToModify=((Bone="dyn_clip_fix_c",Transform=(Rotation=(X=0,Y=0,Z=0,W=1),Translation=(X=0,Y=0.5,Z=0),Scale3D=(X=1,Y=1,Z=1))))
    Mode=AdditiveLocal
    WeightMaximum=1
    WeightMinimum=0
    Weight=&RigVMFunction_MathFloatRemap(
     SourceMaximum=9
     SourceMinimum=7.17
     TargetMinimum=0
     bClamp=true
     Value=&RigUnit_GetRelativeBoneTransform(Bone=C_brow_mid, Space=faceAttach).Transform.Translation.Z
    ).Result
   )
  ]}
]
hierarchy [
 root {loc=<0 0 0>, rot=<0 0 0 1>}
 pelvis {parent=root, loc=<0 -0.7071 93.1191>, rot=<0.0127 0.707 -0.0127 -0.707>}
 spine_01 {parent=pelvis, loc=<6.7494 -0.6388 0>, rot=<0 0 0.034 -0.9994>}
 spine_02 {parent=spine_01, loc=<7.2518 -0.035 0>, rot=<0 0 0.0238 -0.9997>}
 spine_03 {parent=spine_02, loc=<6.0104 0 0>, rot=<0 0 0.0149 -0.9999>}
 spine_04 {parent=spine_03, loc=<8.1024 0 0>, rot=<0 0 -0.0132 -0.9999>}
 spine_05 {parent=spine_04, loc=<10.0146 0 0>, rot=<0 0 -0.0015 -1>}
 clavicle_l {parent=spine_05, loc=<4.3575 -0.1466 -2.3223>, rot=<0.0465 -0.7289 -0.009 -0.6829>}
 clavicle_r {parent=spine_05, loc=<4.3575 -0.1466 2.3223>, rot=<0.7289 0.0465 -0.6829 0.009>}
 neck_01 {parent=spine_05, loc=<11.3138 0.1529 0>, rot=<0 0 -0.2003 -0.9797>}
 neck_02 {parent=neck_01, loc=<4.5627 0 0>, rot=<0 0 0.0504 -0.9987>}
 head {parent=neck_02, loc=<5.3914 0 0>, rot=<0 0 0.115 -0.9934>}
 faceAttach {parent=head, loc=<0.6201 0.9296 0>, rot=<0.0082 -0.7071 -0.0082 -0.7071>, scale=<0.89 0.89 0.89>}
 C_brow_mid {parent=faceAttach, loc=<-0.0046 10.8396 7.1678>, rot=<0 0.0161 -0.0032 -0.9999>}
 L_brow_mid {parent=faceAttach, loc=<3.5874 10.1696 8.2452>, rot=<-0.0005 -0.0048 0.1832 -0.9831>}
 L_brow_outer {parent=faceAttach, loc=<5.8921 7.9031 8.2566>, rot=<-0.0332 -0.1033 0.4335 -0.8946>}
 L_cheek_inner {parent=faceAttach, loc=<4.4945 8.5711 1.9759>, rot=<-0.0663 -0.2589 0.3199 -0.909>}
 C_lip_upper_mid {parent=faceAttach, loc=<-0.0035 11.0855 -0.3085>, rot=<0.0209 0.0007 -0.0004 -0.9998>}
 L_lip_upper_outer {parent=faceAttach, loc=<1.7715 10.1921 -0.2714>, rot=<0.0346 -0.0651 0.2839 -0.956>}
 L_lip_corner {parent=faceAttach, loc=<2.6883 9.377 -0.6059>, rot=<0.0011 0.0084 0.2055 -0.9786>}
 C_nose_bridge {parent=faceAttach, loc=<0 12.0636 3.9809>, rot=<-0.2218 0.007 0.002 -0.9751>}
 R_brow_mid {parent=faceAttach, loc=<-3.5874 10.1696 8.2452>, rot=<-0.0005 0.0048 -0.1832 -0.9831>}
 R_brow_outer {parent=faceAttach, loc=<-5.8921 7.9031 8.2566>, rot=<-0.0332 0.1033 -0.4335 -0.8946>}
 R_cheek_inner {parent=faceAttach, loc=<-4.4945 8.5711 1.9759>, rot=<-0.0663 0.2589 -0.3199 -0.909>}
 R_lip_corner {parent=faceAttach, loc=<-2.6883 9.377 -0.6058>, rot=<0.0011 -0.0084 -0.2055 -0.9786>}
 C_jaw {parent=faceAttach, loc=<-0.0151 1.9161 0.6507>, rot=<0 0 0 1>}
 L_lip_lower_outer {parent=C_jaw, loc=<1.7722 8.1697 -1.2598>, rot=<-0.0131 0.0315 0.2689 -0.9626>}
 C_lip_lower_mid {parent=C_jaw, loc=<0.0084 9.0718 -1.2785>, rot=<0 -0.0248 -0.0053 -0.9997>}
 R_lip_lower_outer {parent=C_jaw, loc=<-1.742 8.1697 -1.2598>, rot=<-0.0131 -0.0315 -0.2689 -0.9626>}
 teeth_lower {parent=C_jaw, loc=<0.0158 7.585 -1.9184>, rot=<0 0 0 1>}
 R_eye {parent=faceAttach, loc=<-3.5817 7.7209 6.7371>, rot=<0 0 0.9998 -0.0187>}
 R_eye_lid_upper_mid {parent=R_eye, loc=<-0.0396 -1.1714 0.2631>, rot=<0.0523 -0.0009 0.0167 -0.9985>}
 R_eye_lid_lower_mid {parent=R_eye, loc=<-0.0085 -0.9425 -0.7052>, rot=<0.0329 -0.0005 0.0167 -0.9993>}
 L_eye {parent=faceAttach, loc=<3.5108 7.7162 6.737>, rot=<0 0 0.9998 0.0187>}
 L_eye_lid_upper_mid {parent=L_eye, loc=<0.0251 -1.1623 0.2388>, rot=<0.0523 0 0 -0.9986>}
 L_eye_lid_lower_mid {parent=L_eye, loc=<0.0223 -0.923 -0.7488>, rot=<0.0329 0 0 -0.9995>}
 R_lip_upper_outer {parent=faceAttach, loc=<-1.7715 10.1921 -0.2714>, rot=<0.0346 0.0651 -0.2839 -0.956>}
 dyn_hair_strap_a_base_l {parent=head, loc=<7.948 -7.2009 -0.2084>, rot=<0.2726 -0.9585 0.0301 -0.0781>}
 dyn_hair_strap_a_l {parent=dyn_hair_strap_a_base_l, loc=<0 0 0>, rot=<0 0 0 -1>}
 dyn_hair_strap_b_l {parent=dyn_hair_strap_a_l, loc=<7 0 0>, rot=<0 0 0 -1>}
 dyn_hair_strap_c_l {parent=dyn_hair_strap_b_l, loc=<7 0 0>, rot=<0 0 0 -1>}
 dyn_hair_strap_d_l {parent=dyn_hair_strap_c_l, loc=<7 0 0>, rot=<0 0 0 -1>}
 dyn_hair_strap_e_l {parent=dyn_hair_strap_d_l, loc=<7 0 0>, rot=<0 0 0 -1>}
 dyn_hair_strap_a_base_r {parent=head, loc=<6.4998 -6.7564 3.0051>, rot=<0.2763 -0.9528 0.0664 0.1064>}
 dyn_hair_strap_a_r {parent=dyn_hair_strap_a_base_r, loc=<0 0 0>, rot=<0 0 0 -1>}
 dyn_hair_strap_b_r {parent=dyn_hair_strap_a_r, loc=<6.2 0 0>, rot=<0 0 0 -1>}
 dyn_hair_strap_c_r {parent=dyn_hair_strap_b_r, loc=<6.2 0 0>, rot=<0 0 0 -1>}
 dyn_hair_strap_d_r {parent=dyn_hair_strap_c_r, loc=<6.2 0 0>, rot=<0 0 0 -1>}
 dyn_hair_strap_e_r {parent=dyn_hair_strap_d_r, loc=<6.2 0 0>, rot=<0 0 0 -1>}
 dyn_hair_a_base {parent=head, loc=<10.8637 10.3249 -3.7623>, rot=<-0.2073 -0.9496 0.0521 0.2295>}
 dyn_hair_a {parent=dyn_hair_a_base, loc=<0 0 0>, rot=<0 0 0 1>}
 dyn_hair_b_base {parent=head, loc=<11.6322 10.3455 -1.6826>, rot=<-0.1465 -0.9608 0.0374 0.2323>}
 dyn_hair_b {parent=dyn_hair_b_base, loc=<0 0 0>, rot=<0 0 0 -1>}
 dyn_hair_c_base {parent=head, loc=<10.5219 10.6679 1.3844>, rot=<-0.0403 -0.965 0.0109 0.259>}
 dyn_hair_c {parent=dyn_hair_c_base, loc=<0 0 0>, rot=<0 0 0 -1>}
 dyn_hair_d_base {parent=head, loc=<11.2633 11.0014 3.1032>, rot=<0.1207 -0.9502 -0.0817 0.2755>}
 dyn_hair_d {parent=dyn_hair_d_base, loc=<0 0 0>, rot=<0 0 0 1>}
 dyn_hair_e_base {parent=head, loc=<12.9748 11.1343 -1.4792>, rot=<-0.13 -0.9016 -0.001 0.4126>}
 dyn_hair_e {parent=dyn_hair_e_base, loc=<0 0 0>, rot=<0 0 0 1>}
 dyn_hair_f_base {parent=head, loc=<14.5735 10.3958 -0.3575>, rot=<-0.3587 -0.9152 -0.0731 0.1686>}
 dyn_hair_f {parent=dyn_hair_f_base, loc=<0 0 0>, rot=<0 0 0 1>}
 dyn_hair_g_base {parent=head, loc=<15.7372 8.6309 3.0721>, rot=<-0.4015 -0.8542 -0.2886 0.1609>}
 dyn_hair_g {parent=dyn_hair_g_base, loc=<0 0 0>, rot=<0 0 0 -1>}
 dyn_hair_h_base {parent=head, loc=<13.7403 7.8503 5.631>, rot=<-0.3862 -0.8151 -0.4114 0.1314>}
 dyn_hair_h {parent=dyn_hair_h_base, loc=<0 0 0>, rot=<0 0 0 1>}
 dyn_hair_i_base {parent=head, loc=<12.222 6.9206 6.4052>, rot=<0.1119 -0.8204 -0.5428 0.1405>}
 dyn_hair_i {parent=dyn_hair_i_base, loc=<0 0 0>, rot=<0 0 0 1>}
 dyn_hair_j_base {parent=head, loc=<13.9541 6.753 -4.5804>, rot=<0.8108 0.5378 -0.1873 -0.1353>}
 dyn_hair_j {parent=dyn_hair_j_base, loc=<0 0 0>, rot=<0 0 0 1>}
 dyn_hair_k_base {parent=head, loc=<15.4691 4.6721 -3.9485>, rot=<0.9727 0.0228 -0.2306 -0.0146>}
 dyn_hair_k {parent=dyn_hair_k_base, loc=<0 0 0>, rot=<0 0 0 -1>}
 dyn_hair_base_l {parent=head, loc=<16.0205 7.3722 -0.4459>, rot=<0.7802 0.5603 0.0725 -0.2685>}
 dyn_hair_l {parent=dyn_hair_base_l, loc=<0 0 0>, rot=<0 0 0 1>}
 dyn_hair_m_base {parent=head, loc=<16.4764 5.1128 -1.2661>, rot=<0.9614 0.0127 0.2737 -0.0238>}
 dyn_hair_m {parent=dyn_hair_m_base, loc=<0 0 0>, rot=<0 0 0 1>}
 dyn_hair_n_base {parent=head, loc=<16.6815 3.3257 -3.4676>, rot=<0.9415 0.16 0.169 0.2437>}
 dyn_hair_n {parent=dyn_hair_n_base, loc=<0 0 0>, rot=<0 0 0 1>}
 dyn_hair_o_base {parent=head, loc=<13.7275 5.0377 -5.53>, rot=<-0.6412 -0.014 0.7671 -0.0175>}
 dyn_hair_o {parent=dyn_hair_o_base, loc=<0 0 0>, rot=<0 0 0 -1>}
 dyn_hair_p_base {parent=head, loc=<16.9595 4.7548 1.0783>, rot=<0.8439 0.3588 0.384 0.1082>}
 dyn_hair_p {parent=dyn_hair_p_base, loc=<0 0 0>, rot=<0 0 0 1>}
 dyn_hair_q_base {parent=head, loc=<16.512 2.9803 3.4247>, rot=<0.7313 -0.0963 0.6496 -0.1843>}
 dyn_hair_q {parent=dyn_hair_q_base, loc=<0 0 0>, rot=<0 0 0 1>}
 dyn_hair_base_r {parent=head, loc=<15.1359 1.7331 5.9706>, rot=<0.6262 -0.1227 0.7515 -0.1678>}
 dyn_hair_r {parent=dyn_hair_base_r, loc=<0 0 0>, rot=<0 0 0 1>}
 dyn_hair_s_base {parent=head, loc=<16.4276 -0.3428 -4.1461>, rot=<0.9661 -0.1194 0.1434 0.1784>}
 dyn_hair_s {parent=dyn_hair_s_base, loc=<0 0 0>, rot=<0 0 0 1>}
 dyn_hair_t_base {parent=head, loc=<17.6384 1.082 -1.3116>, rot=<-0.7644 0.3327 -0.2183 -0.5074>}
 dyn_hair_t {parent=dyn_hair_t_base, loc=<0 0 0>, rot=<0 0 0 1>}
 dyn_hair_u_base {parent=head, loc=<14.9685 1.254 -6.0014>, rot=<0.8913 -0.3543 -0.2134 0.1858>}
 dyn_hair_u {parent=dyn_hair_u_base, loc=<0 0 0>, rot=<0 0 0 1>}
 dyn_hair_v_base {parent=head, loc=<13.5141 -0.7284 -6.1133>, rot=<0.6687 -0.6098 -0.3917 0.1661>}
 dyn_hair_v {parent=dyn_hair_v_base, loc=<0 0 0>, rot=<0 0 0 1>}
 dyn_hair_w_base {parent=head, loc=<14.3109 2.9925 -5.9979>, rot=<-0.5739 -0.0773 0.6133 -0.5371>}
 dyn_hair_w {parent=dyn_hair_w_base, loc=<0 0 0>, rot=<0 0 0 1>}
 dyn_hair_x_base {parent=head, loc=<17.051 -0.1007 1.1821>, rot=<0.8439 -0.4492 0.2933 0.0003>}
 dyn_hair_x {parent=dyn_hair_x_base, loc=<0 0 0>, rot=<0 0 0 1>}
 dyn_hair_y_base {parent=head, loc=<16.4807 0.3738 3.4631>, rot=<0.8258 -0.1046 0.5309 -0.1588>}
 dyn_hair_y {parent=dyn_hair_y_base, loc=<0 0 0>, rot=<0 0 0 1>}
 dyn_hair_z_base {parent=head, loc=<15.1757 -0.9564 4.4595>, rot=<0.6675 -0.0843 0.6869 -0.2749>}
 dyn_hair_z {parent=dyn_hair_z_base, loc=<0 0 0>, rot=<0 0 0 1>}
 dyn_hair_aa_base {parent=head, loc=<13.4849 -2.136 6.121>, rot=<0.3402 -0.1552 0.8737 -0.311>}
 dyn_hair_aa {parent=dyn_hair_aa_base, loc=<0 0 0>, rot=<0 0 0 1>}
 dyn_hairbb_base {parent=head, loc=<11.9474 -4.2803 5.4584>, rot=<0.2438 -0.2207 0.8695 -0.3684>}
 dyn_hairbb {parent=dyn_hairbb_base, loc=<0 0 0>, rot=<0 0 0 1>}
 dyn_hair_cc_base {parent=head, loc=<15.4309 -3.2727 2.2593>, rot=<0.6778 -0.0269 0.5719 -0.4612>}
 dyn_hair_cc {parent=dyn_hair_cc_base, loc=<0 0 0>, rot=<0 0 0 -1>}
 dyn_hair_dd_base {parent=head, loc=<15.6684 -3.3327 -1.9697>, rot=<0.6089 -0.1214 0.6178 -0.4826>}
 dyn_hair_dd {parent=dyn_hair_dd_base, loc=<0 0 0>, rot=<0 0 0 1>}
 dyn_hair_ee_base {parent=head, loc=<13.3013 -3.9035 -4.2369>, rot=<0.2315 -0.6559 0.3898 -0.6036>}
 dyn_hair_ee {parent=dyn_hair_ee_base, loc=<0 0 0>, rot=<0 0 0 1>}
 dyn_hair_ff_base {parent=head, loc=<13.2119 -5.3269 -1.8854>, rot=<0.6482 -0.4009 0.5352 -0.3642>}
 dyn_hair_ff {parent=dyn_hair_ff_base, loc=<0 0 0>, rot=<0 0 0 1>}
 dyn_hair_gg_base {parent=head, loc=<11.4532 -6.0224 0.495>, rot=<0.717 -0.2583 0.5989 -0.2458>}
 dyn_hair_gg {parent=dyn_hair_gg_base, loc=<0 0 0>, rot=<0 0 0 1>}
 dyn_hair_hh_base {parent=head, loc=<11.0275 -5.6466 2.8723>, rot=<0.4343 -0.4594 0.7567 -0.1668>}
 dyn_hair_hh {parent=dyn_hair_hh_base, loc=<0 0 0>, rot=<0 0 0 1>}
 dyn_hair_ii_base {parent=head, loc=<12.6372 -5.4202 2.7553>, rot=<0.6951 -0.2742 0.6564 -0.1034>}
 dyn_hair_ii {parent=dyn_hair_ii_base, loc=<0 0 0>, rot=<0 0 0 1>}
 dyn_hair_jj_base {parent=head, loc=<12.0883 5.5016 -6.0353>, rot=<-0.2959 -0.6886 0.5437 -0.3777>}
 dyn_hair_jj {parent=dyn_hair_jj_base, loc=<0 0 0>, rot=<0 0 0 -1>}
 dyn_hair_kk_base {parent=head, loc=<11.929 2.2296 7.3309>, rot=<0.3402 -0.1552 0.8737 -0.311>}
 dyn_hair_kk {parent=dyn_hair_kk_base, loc=<0 0 0>, rot=<0 0 0 1>}
 dyn_clip_fix_l {parent=head, loc=<9.8492 9.9925 -3.136>, rot=<0.1372 -0.697 0.1219 -0.6932>}
 dyn_clip_fix_r {parent=head, loc=<9.8492 9.9925 3.136>, rot=<-0.1214 -0.6962 -0.1378 -0.694>}
 dyn_clip_fix_c {parent=head, loc=<10.0975 10.5386 0.1096>, rot=<0.0058 -0.6956 -0.0105 -0.7184>}
 dyn_root_simspace {parent=root, loc=<0 0 0>, rot=<0 0 0 1>}
]
