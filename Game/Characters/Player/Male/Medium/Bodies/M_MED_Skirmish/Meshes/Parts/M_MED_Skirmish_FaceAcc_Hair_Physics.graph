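# Physics asset for the Skirmish medium male player's FaceAcc hair mesh, defining capsule bodies and constraints for its dyn_hair bones.
# Each of the many strands pairs a small base capsule with a longer hair capsule, joined by tight swing and twist limits, slightly wider in one swing axis for strands a–d.
# Every base connects directly to one large head capsule with free swing and twist, forming a radial set of short two-body chains.
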PhysicsAsset /Game/Characters/Player/Male/Medium/Bodies/M_MED_Skirmish/Meshes/Parts/M_MED_Skirmish_FaceAcc_Hair_Physics
bodies {
 dyn_hair_a [{shapeType=Capsule, length=2.5, radius=0.75, center=<2 0 0>, rotation=<rot 90 0 0>}]
 dyn_hair_a_base [{shapeType=Capsule, length=0.505, radius=0.505, rotation=<rot 0 -90 0>}]
 dyn_hair_aa [{shapeType=Capsule, length=2.5, radius=0.75, center=<2 0 0>, rotation=<rot 90 0 0>}]
 dyn_hair_aa_base [{shapeType=Capsule, length=0.505, radius=0.505, rotation=<rot 0 -90 0>}]
 dyn_hair_b [{shapeType=Capsule, length=2.5, radius=0.75, center=<2 0 0>, rotation=<rot 90 0 0>}]
 dyn_hair_b_base [{shapeType=Capsule, length=0.505, radius=0.505, rotation=<rot 0 -90 0>}]
 dyn_hair_base_l [{shapeType=Capsule, length=0.505, radius=0.505, rotation=<rot 0 -90 0>}]
 dyn_hair_base_r [{shapeType=Capsule, length=0.505, radius=0.505, rotation=<rot 0 -90 0>}]
 dyn_hair_c [{shapeType=Capsule, length=2.5, radius=0.75, center=<2 0 0>, rotation=<rot 90 0 0>}]
 dyn_hair_c_base [{shapeType=Capsule, length=0.505, radius=0.505, rotation=<rot 0 -90 0>}]
 dyn_hair_cc [{shapeType=Capsule, length=2.5, radius=0.75, center=<2 0 0>, rotation=<rot 90 0 0>}]
 dyn_hair_cc_base [{shapeType=Capsule, length=0.505, radius=0.505, rotation=<rot 0 -90 0>}]
 dyn_hair_d [{shapeType=Capsule, length=2.5, radius=0.75, center=<2 0 0>, rotation=<rot 90 0 0>}]
 dyn_hair_d_base [{shapeType=Capsule, length=0.505, radius=0.505, rotation=<rot 0 -90 0>}]
 dyn_hair_dd [{shapeType=Capsule, length=2.5, radius=0.75, center=<2 0 0>, rotation=<rot 90 0 0>}]
 dyn_hair_dd_base [{shapeType=Capsule, length=0.505, radius=0.505, rotation=<rot 0 -90 0>}]
 dyn_hair_e [{shapeType=Capsule, length=2.5, radius=0.75, center=<2 0 0>, rotation=<rot 90 0 0>}]
 dyn_hair_e_base [{shapeType=Capsule, length=0.505, radius=0.505, rotation=<rot 0 -90 0>}]
 dyn_hair_ee [{shapeType=Capsule, length=2.5, radius=0.75, center=<2 0 0>, rotation=<rot 90 0 0>}]
 dyn_hair_ee_base [{shapeType=Capsule, length=0.505, radius=0.505, rotation=<rot 0 -90 0>}]
 dyn_hair_f [{shapeType=Capsule, length=2.5, radius=0.75, center=<2 0 0>, rotation=<rot 90 0 0>}]
 dyn_hair_f_base [{shapeType=Capsule, length=0.505, radius=0.505, rotation=<rot 0 -90 0>}]
 dyn_hair_ff [{shapeType=Capsule, length=2.5, radius=0.75, center=<2 0 0>, rotation=<rot 90 0 0>}]
 dyn_hair_ff_base [{shapeType=Capsule, length=0.505, radius=0.505, rotation=<rot 0 -90 0>}]
 dyn_hair_g [{shapeType=Capsule, length=2.5, radius=0.75, center=<2 0 0>, rotation=<rot 90 0 0>}]
 dyn_hair_g_base [{shapeType=Capsule, length=0.505, radius=0.505, rotation=<rot 0 -90 0>}]
 dyn_hair_gg [{shapeType=Capsule, length=2.5, radius=0.75, center=<2 0 0>, rotation=<rot 90 0 0>}]
 dyn_hair_gg_base [{shapeType=Capsule, length=0.505, radius=0.505, rotation=<rot 0 -90 0>}]
 dyn_hair_h [{shapeType=Capsule, length=2.5, radius=0.75, center=<2 0 0>, rotation=<rot 90 0 0>}]
 dyn_hair_h_base [{shapeType=Capsule, length=0.505, radius=0.505, rotation=<rot 0 -90 0>}]
 dyn_hair_hh [{shapeType=Capsule, length=2.5, radius=0.75, center=<2 0 0>, rotation=<rot 90 0 0>}]
 dyn_hair_hh_base [{shapeType=Capsule, length=0.505, radius=0.505, rotation=<rot 0 -90 0>}]
 dyn_hair_i [{shapeType=Capsule, length=2.5, radius=0.75, center=<2 0 0>, rotation=<rot 90 0 0>}]
 dyn_hair_i_base [{shapeType=Capsule, length=0.505, radius=0.505, rotation=<rot 0 -90 0>}]
 dyn_hair_ii [{shapeType=Capsule, length=2.5, radius=0.75, center=<2 0 0>, rotation=<rot 90 0 0>}]
 dyn_hair_ii_base [{shapeType=Capsule, length=0.505, radius=0.505, rotation=<rot 0 -90 0>}]
 dyn_hair_j [{shapeType=Capsule, length=2.5, radius=0.75, center=<2 0 0>, rotation=<rot 90 0 0>}]
 dyn_hair_j_base [{shapeType=Capsule, length=0.505, radius=0.505, rotation=<rot 0 -90 0>}]
 dyn_hair_jj [{shapeType=Capsule, length=2.5, radius=0.75, center=<2 0 0>, rotation=<rot 90 0 0>}]
 dyn_hair_jj_base [{shapeType=Capsule, length=0.505, radius=0.505, rotation=<rot 0 -90 0>}]
 dyn_hair_k [{shapeType=Capsule, length=2.5, radius=0.75, center=<2 0 0>, rotation=<rot 90 0 0>}]
 dyn_hair_k_base [{shapeType=Capsule, length=0.505, radius=0.505, rotation=<rot 0 -90 0>}]
 dyn_hair_kk [{shapeType=Capsule, length=2.5, radius=0.75, center=<2 0 0>, rotation=<rot 90 0 0>}]
 dyn_hair_kk_base [{shapeType=Capsule, length=0.505, radius=0.505, rotation=<rot 0 -90 0>}]
 dyn_hair_l [{shapeType=Capsule, length=2.5, radius=0.75, center=<2 0 0>, rotation=<rot 90 0 0>}]
 dyn_hair_m [{shapeType=Capsule, length=2.5, radius=0.75, center=<2 0 0>, rotation=<rot 90 0 0>}]
 dyn_hair_m_base [{shapeType=Capsule, length=0.505, radius=0.505, rotation=<rot 0 -90 0>}]
 dyn_hair_n [{shapeType=Capsule, length=2.5, radius=0.75, center=<2 0 0>, rotation=<rot 90 0 0>}]
 dyn_hair_n_base [{shapeType=Capsule, length=0.505, radius=0.505, rotation=<rot 0 -90 0>}]
 dyn_hair_o [{shapeType=Capsule, length=2.5, radius=0.75, center=<2 0 0>, rotation=<rot 90 0 0>}]
 dyn_hair_o_base [{shapeType=Capsule, length=0.505, radius=0.505, rotation=<rot 0 -90 0>}]
 dyn_hair_p [{shapeType=Capsule, length=2.5, radius=0.75, center=<2 0 0>, rotation=<rot 90 0 0>}]
 dyn_hair_p_base [{shapeType=Capsule, length=0.505, radius=0.505, rotation=<rot 0 -90 0>}]
 dyn_hair_q [{shapeType=Capsule, length=2.5, radius=0.75, center=<2 0 0>, rotation=<rot 90 0 0>}]
 dyn_hair_q_base [{shapeType=Capsule, length=0.505, radius=0.505, rotation=<rot 0 -90 0>}]
 dyn_hair_r [{shapeType=Capsule, length=2.5, radius=0.75, center=<2 0 0>, rotation=<rot 90 0 0>}]
 dyn_hair_s [{shapeType=Capsule, length=2.5, radius=0.75, center=<2 0 0>, rotation=<rot 90 0 0>}]
 dyn_hair_s_base [{shapeType=Capsule, length=0.505, radius=0.505, rotation=<rot 0 -90 0>}]
 dyn_hair_t [{shapeType=Capsule, length=2.5, radius=0.75, center=<2 0 0>, rotation=<rot 90 0 0>}]
 dyn_hair_t_base [{shapeType=Capsule, length=0.505, radius=0.505, rotation=<rot 0 -90 0>}]
 dyn_hair_u [{shapeType=Capsule, length=2.5, radius=0.75, center=<2 0 0>, rotation=<rot 90 0 0>}]
 dyn_hair_u_base [{shapeType=Capsule, length=0.505, radius=0.505, rotation=<rot 0 -90 0>}]
 dyn_hair_v [{shapeType=Capsule, length=2.5, radius=0.75, center=<2 0 0>, rotation=<rot 90 0 0>}]
 dyn_hair_v_base [{shapeType=Capsule, length=0.505, radius=0.505, rotation=<rot 0 -90 0>}]
 dyn_hair_w [{shapeType=Capsule, length=2.5, radius=0.75, center=<2 0 0>, rotation=<rot 90 0 0>}]
 dyn_hair_w_base [{shapeType=Capsule, length=0.505, radius=0.505, rotation=<rot 0 -90 0>}]
 dyn_hair_x [{shapeType=Capsule, length=2.5, radius=0.75, center=<2 0 0>, rotation=<rot 90 0 0>}]
 dyn_hair_x_base [{shapeType=Capsule, length=0.505, radius=0.505, rotation=<rot 0 -90 0>}]
 dyn_hair_y [{shapeType=Capsule, length=2.5, radius=0.75, center=<2 0 0>, rotation=<rot 90 0 0>}]
 dyn_hair_y_base [{shapeType=Capsule, length=0.505, radius=0.505, rotation=<rot 0 -90 0>}]
 dyn_hair_z [{shapeType=Capsule, length=2.5, radius=0.75, center=<2 0 0>, rotation=<rot 90 0 0>}]
 dyn_hair_z_base [{shapeType=Capsule, length=0.505, radius=0.505, rotation=<rot 0 -90 0>}]
 dyn_hairbb [{shapeType=Capsule, length=2.5, radius=0.75, center=<2 0 0>, rotation=<rot 90 0 0>}]
 dyn_hairbb_base [{shapeType=Capsule, length=0.505, radius=0.505, rotation=<rot 0 -90 0>}]
 head [{shapeType=Capsule, length=4, radius=8.25, center=<5.6008 2.7159 0>, rotation=<rot 90 0 0>}]
}
constraints [
 {bone1Name=dyn_hair_a, bone2Name=dyn_hair_a_base, swing1LimitDegrees=10, swing2LimitDegrees=5, twistLimitDegrees=5}
 {bone1Name=dyn_hair_b, bone2Name=dyn_hair_b_base, swing1LimitDegrees=10, swing2LimitDegrees=5, twistLimitDegrees=5}
 {bone1Name=dyn_hair_c, bone2Name=dyn_hair_c_base, swing1LimitDegrees=10, swing2LimitDegrees=5, twistLimitDegrees=5}
 {bone1Name=dyn_hair_d, bone2Name=dyn_hair_d_base, swing1LimitDegrees=10, swing2LimitDegrees=5, twistLimitDegrees=5}
 {bone1Name=dyn_hair_e, bone2Name=dyn_hair_e_base, swing1LimitDegrees=5, swing2LimitDegrees=5, twistLimitDegrees=5}
 {bone1Name=dyn_hair_f, bone2Name=dyn_hair_f_base, swing1LimitDegrees=5, swing2LimitDegrees=5, twistLimitDegrees=5}
 {bone1Name=dyn_hair_g, bone2Name=dyn_hair_g_base, swing1LimitDegrees=5, swing2LimitDegrees=5, twistLimitDegrees=5}
 {bone1Name=dyn_hair_h, bone2Name=dyn_hair_h_base, swing1LimitDegrees=5, swing2LimitDegrees=5, twistLimitDegrees=5}
 {bone1Name=dyn_hair_i, bone2Name=dyn_hair_i_base, swing1LimitDegrees=5, swing2LimitDegrees=5, twistLimitDegrees=5}
 {bone1Name=dyn_hair_j, bone2Name=dyn_hair_j_base, swing1LimitDegrees=5, swing2LimitDegrees=5, twistLimitDegrees=5}
 {bone1Name=dyn_hair_k, bone2Name=dyn_hair_k_base, swing1LimitDegrees=5, swing2LimitDegrees=5, twistLimitDegrees=5}
 {bone1Name=dyn_hair_l, bone2Name=dyn_hair_base_l, swing1LimitDegrees=5, swing2LimitDegrees=5, twistLimitDegrees=5}
 {bone1Name=dyn_hair_m, bone2Name=dyn_hair_m_base, swing1LimitDegrees=5, swing2LimitDegrees=5, twistLimitDegrees=5}
 {bone1Name=dyn_hair_n, bone2Name=dyn_hair_n_base, swing1LimitDegrees=5, swing2LimitDegrees=5, twistLimitDegrees=5}
 {bone1Name=dyn_hair_o, bone2Name=dyn_hair_o_base, swing1LimitDegrees=5, swing2LimitDegrees=5, twistLimitDegrees=5}
 {bone1Name=dyn_hair_p, bone2Name=dyn_hair_p_base, swing1LimitDegrees=5, swing2LimitDegrees=5, twistLimitDegrees=5}
 {bone1Name=dyn_hair_q, bone2Name=dyn_hair_q_base, swing1LimitDegrees=5, swing2LimitDegrees=5, twistLimitDegrees=5}
 {bone1Name=dyn_hair_r, bone2Name=dyn_hair_base_r, swing1LimitDegrees=5, swing2LimitDegrees=5, twistLimitDegrees=5}
 {bone1Name=dyn_hair_s, bone2Name=dyn_hair_s_base, swing1LimitDegrees=5, swing2LimitDegrees=5, twistLimitDegrees=5}
 {bone1Name=dyn_hair_t, bone2Name=dyn_hair_t_base, swing1LimitDegrees=5, swing2LimitDegrees=5, twistLimitDegrees=5}
 {bone1Name=dyn_hair_u, bone2Name=dyn_hair_u_base, swing1LimitDegrees=5, swing2LimitDegrees=5, twistLimitDegrees=5}
 {bone1Name=dyn_hair_v, bone2Name=dyn_hair_v_base, swing1LimitDegrees=5, swing2LimitDegrees=5, twistLimitDegrees=5}
 {bone1Name=dyn_hair_w, bone2Name=dyn_hair_w_base, swing1LimitDegrees=5, swing2LimitDegrees=5, twistLimitDegrees=5}
 {bone1Name=dyn_hair_x, bone2Name=dyn_hair_x_base, swing1LimitDegrees=5, swing2LimitDegrees=5, twistLimitDegrees=5}
 {bone1Name=dyn_hair_y, bone2Name=dyn_hair_y_base, swing1LimitDegrees=5, swing2LimitDegrees=5, twistLimitDegrees=5}
 {bone1Name=dyn_hair_z, bone2Name=dyn_hair_z_base, swing1LimitDegrees=5, swing2LimitDegrees=5, twistLimitDegrees=5}
 {bone1Name=dyn_hair_aa, bone2Name=dyn_hair_aa_base, swing1LimitDegrees=5, swing2LimitDegrees=5, twistLimitDegrees=5}
 {bone1Name=dyn_hairbb, bone2Name=dyn_hairbb_base, swing1LimitDegrees=5, swing2LimitDegrees=5, twistLimitDegrees=5}
 {bone1Name=dyn_hair_cc, bone2Name=dyn_hair_cc_base, swing1LimitDegrees=5, swing2LimitDegrees=5, twistLimitDegrees=5}
 {bone1Name=dyn_hair_dd, bone2Name=dyn_hair_dd_base, swing1LimitDegrees=5, swing2LimitDegrees=5, twistLimitDegrees=5}
 {bone1Name=dyn_hair_ee, bone2Name=dyn_hair_ee_base, swing1LimitDegrees=5, swing2LimitDegrees=5, twistLimitDegrees=5}
 {bone1Name=dyn_hair_ff, bone2Name=dyn_hair_ff_base, swing1LimitDegrees=5, swing2LimitDegrees=5, twistLimitDegrees=5}
 {bone1Name=dyn_hair_gg, bone2Name=dyn_hair_gg_base, swing1LimitDegrees=5, swing2LimitDegrees=5, twistLimitDegrees=5}
 {bone1Name=dyn_hair_hh, bone2Name=dyn_hair_hh_base, swing1LimitDegrees=5, swing2LimitDegrees=5, twistLimitDegrees=5}
 {bone1Name=dyn_hair_ii, bone2Name=dyn_hair_ii_base, swing1LimitDegrees=5, swing2LimitDegrees=5, twistLimitDegrees=5}
 {bone1Name=dyn_hair_jj, bone2Name=dyn_hair_jj_base, swing1LimitDegrees=5, swing2LimitDegrees=5, twistLimitDegrees=5}
 {bone1Name=dyn_hair_kk, bone2Name=dyn_hair_kk_base, swing1LimitDegrees=5, swing2LimitDegrees=5, twistLimitDegrees=5}
 {
  bone1Name=dyn_hair_a_base
  bone2Name=head
  swing1LimitDegrees=45
  swing1Motion=Free
  swing2LimitDegrees=45
  swing2Motion=Free
  twistLimitDegrees=45
  twistMotion=Free
 }
 {
  bone1Name=dyn_hair_b_base
  bone2Name=head
  swing1LimitDegrees=45
  swing1Motion=Free
  swing2LimitDegrees=45
  swing2Motion=Free
  twistLimitDegrees=45
  twistMotion=Free
 }
 {
  bone1Name=dyn_hair_c_base
  bone2Name=head
  swing1LimitDegrees=45
  swing1Motion=Free
  swing2LimitDegrees=45
  swing2Motion=Free
  twistLimitDegrees=45
  twistMotion=Free
 }
 {
  bone1Name=dyn_hair_d_base
  bone2Name=head
  swing1LimitDegrees=45
  swing1Motion=Free
  swing2LimitDegrees=45
  swing2Motion=Free
  twistLimitDegrees=45
  twistMotion=Free
 }
 {
  bone1Name=dyn_hair_e_base
  bone2Name=head
  swing1LimitDegrees=45
  swing1Motion=Free
  swing2LimitDegrees=45
  swing2Motion=Free
  twistLimitDegrees=45
  twistMotion=Free
 }
 {
  bone1Name=dyn_hair_f_base
  bone2Name=head
  swing1LimitDegrees=45
  swing1Motion=Free
  swing2LimitDegrees=45
  swing2Motion=Free
  twistLimitDegrees=45
  twistMotion=Free
 }
 {
  bone1Name=dyn_hair_g_base
  bone2Name=head
  swing1LimitDegrees=45
  swing1Motion=Free
  swing2LimitDegrees=45
  swing2Motion=Free
  twistLimitDegrees=45
  twistMotion=Free
 }
 {
  bone1Name=dyn_hair_h_base
  bone2Name=head
  swing1LimitDegrees=45
  swing1Motion=Free
  swing2LimitDegrees=45
  swing2Motion=Free
  twistLimitDegrees=45
  twistMotion=Free
 }
 {
  bone1Name=dyn_hair_i_base
  bone2Name=head
  swing1LimitDegrees=45
  swing1Motion=Free
  swing2LimitDegrees=45
  swing2Motion=Free
  twistLimitDegrees=45
  twistMotion=Free
 }
 {
  bone1Name=dyn_hair_j_base
  bone2Name=head
  swing1LimitDegrees=45
  swing1Motion=Free
  swing2LimitDegrees=45
  swing2Motion=Free
  twistLimitDegrees=45
  twistMotion=Free
 }
 {
  bone1Name=dyn_hair_k_base
  bone2Name=head
  swing1LimitDegrees=45
  swing1Motion=Free
  swing2LimitDegrees=45
  swing2Motion=Free
  twistLimitDegrees=45
  twistMotion=Free
 }
 {
  bone1Name=dyn_hair_base_l
  bone2Name=head
  swing1LimitDegrees=45
  swing1Motion=Free
  swing2LimitDegrees=45
  swing2Motion=Free
  twistLimitDegrees=45
  twistMotion=Free
 }
 {
  bone1Name=dyn_hair_m_base
  bone2Name=head
  swing1LimitDegrees=45
  swing1Motion=Free
  swing2LimitDegrees=45
  swing2Motion=Free
  twistLimitDegrees=45
  twistMotion=Free
 }
 {
  bone1Name=dyn_hair_n_base
  bone2Name=head
  swing1LimitDegrees=45
  swing1Motion=Free
  swing2LimitDegrees=45
  swing2Motion=Free
  twistLimitDegrees=45
  twistMotion=Free
 }
 {
  bone1Name=dyn_hair_o_base
  bone2Name=head
  swing1LimitDegrees=45
  swing1Motion=Free
  swing2LimitDegrees=45
  swing2Motion=Free
  twistLimitDegrees=45
  twistMotion=Free
 }
 {
  bone1Name=dyn_hair_p_base
  bone2Name=head
  swing1LimitDegrees=45
  swing1Motion=Free
  swing2LimitDegrees=45
  swing2Motion=Free
  twistLimitDegrees=45
  twistMotion=Free
 }
 {
  bone1Name=dyn_hair_q_base
  bone2Name=head
  swing1LimitDegrees=45
  swing1Motion=Free
  swing2LimitDegrees=45
  swing2Motion=Free
  twistLimitDegrees=45
  twistMotion=Free
 }
 {
  bone1Name=dyn_hair_base_r
  bone2Name=head
  swing1LimitDegrees=45
  swing1Motion=Free
  swing2LimitDegrees=45
  swing2Motion=Free
  twistLimitDegrees=45
  twistMotion=Free
 }
 {
  bone1Name=dyn_hair_s_base
  bone2Name=head
  swing1LimitDegrees=45
  swing1Motion=Free
  swing2LimitDegrees=45
  swing2Motion=Free
  twistLimitDegrees=45
  twistMotion=Free
 }
 {
  bone1Name=dyn_hair_t_base
  bone2Name=head
  swing1LimitDegrees=45
  swing1Motion=Free
  swing2LimitDegrees=45
  swing2Motion=Free
  twistLimitDegrees=45
  twistMotion=Free
 }
 {
  bone1Name=dyn_hair_u_base
  bone2Name=head
  swing1LimitDegrees=45
  swing1Motion=Free
  swing2LimitDegrees=45
  swing2Motion=Free
  twistLimitDegrees=45
  twistMotion=Free
 }
 {
  bone1Name=dyn_hair_v_base
  bone2Name=head
  swing1LimitDegrees=45
  swing1Motion=Free
  swing2LimitDegrees=45
  swing2Motion=Free
  twistLimitDegrees=45
  twistMotion=Free
 }
 {
  bone1Name=dyn_hair_w_base
  bone2Name=head
  swing1LimitDegrees=45
  swing1Motion=Free
  swing2LimitDegrees=45
  swing2Motion=Free
  twistLimitDegrees=45
  twistMotion=Free
 }
 {
  bone1Name=dyn_hair_x_base
  bone2Name=head
  swing1LimitDegrees=45
  swing1Motion=Free
  swing2LimitDegrees=45
  swing2Motion=Free
  twistLimitDegrees=45
  twistMotion=Free
 }
 {
  bone1Name=dyn_hair_y_base
  bone2Name=head
  swing1LimitDegrees=45
  swing1Motion=Free
  swing2LimitDegrees=45
  swing2Motion=Free
  twistLimitDegrees=45
  twistMotion=Free
 }
 {
  bone1Name=dyn_hair_z_base
  bone2Name=head
  swing1LimitDegrees=45
  swing1Motion=Free
  swing2LimitDegrees=45
  swing2Motion=Free
  twistLimitDegrees=45
  twistMotion=Free
 }
 {
  bone1Name=dyn_hair_aa_base
  bone2Name=head
  swing1LimitDegrees=45
  swing1Motion=Free
  swing2LimitDegrees=45
  swing2Motion=Free
  twistLimitDegrees=45
  twistMotion=Free
 }
 {
  bone1Name=dyn_hairbb_base
  bone2Name=head
  swing1LimitDegrees=45
  swing1Motion=Free
  swing2LimitDegrees=45
  swing2Motion=Free
  twistLimitDegrees=45
  twistMotion=Free
 }
 {
  bone1Name=dyn_hair_cc_base
  bone2Name=head
  swing1LimitDegrees=45
  swing1Motion=Free
  swing2LimitDegrees=45
  swing2Motion=Free
  twistLimitDegrees=45
  twistMotion=Free
 }
 {
  bone1Name=dyn_hair_dd_base
  bone2Name=head
  swing1LimitDegrees=45
  swing1Motion=Free
  swing2LimitDegrees=45
  swing2Motion=Free
  twistLimitDegrees=45
  twistMotion=Free
 }
 {
  bone1Name=dyn_hair_ee_base
  bone2Name=head
  swing1LimitDegrees=45
  swing1Motion=Free
  swing2LimitDegrees=45
  swing2Motion=Free
  twistLimitDegrees=45
  twistMotion=Free
 }
 {
  bone1Name=dyn_hair_ff_base
  bone2Name=head
  swing1LimitDegrees=45
  swing1Motion=Free
  swing2LimitDegrees=45
  swing2Motion=Free
  twistLimitDegrees=45
  twistMotion=Free
 }
 {
  bone1Name=dyn_hair_gg_base
  bone2Name=head
  swing1LimitDegrees=45
  swing1Motion=Free
  swing2LimitDegrees=45
  swing2Motion=Free
  twistLimitDegrees=45
  twistMotion=Free
 }
 {
  bone1Name=dyn_hair_hh_base
  bone2Name=head
  swing1LimitDegrees=45
  swing1Motion=Free
  swing2LimitDegrees=45
  swing2Motion=Free
  twistLimitDegrees=45
  twistMotion=Free
 }
 {
  bone1Name=dyn_hair_ii_base
  bone2Name=head
  swing1LimitDegrees=45
  swing1Motion=Free
  swing2LimitDegrees=45
  swing2Motion=Free
  twistLimitDegrees=45
  twistMotion=Free
 }
 {
  bone1Name=dyn_hair_jj_base
  bone2Name=head
  swing1LimitDegrees=45
  swing1Motion=Free
  swing2LimitDegrees=45
  swing2Motion=Free
  twistLimitDegrees=45
  twistMotion=Free
 }
 {
  bone1Name=dyn_hair_kk_base
  bone2Name=head
  swing1LimitDegrees=45
  swing1Motion=Free
  swing2LimitDegrees=45
  swing2Motion=Free
  twistLimitDegrees=45
  twistMotion=Free
 }
]
