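# Physics asset for the Tessel face accessory of the male medium Skirmish body, setting up two dangling hair-strap bone chains.
# Left and right chains each run from a small base capsule through five capsule segments, linked by constraints with moderate swing and tight twist limits.
# Both bases attach to the head, and the head to spine_05, with free angular motion.
# Large head and upper-spine capsules represent the body around the straps.
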PhysicsAsset /Game/Characters/Player/Male/Medium/Bodies/M_MED_Skirmish/Meshes/Parts/M_MED_Skirmish_FaceAcc_Tessel_Physics
bodies {
 dyn_hair_strap_a_base_l [{shapeType=Capsule, length=0.505, radius=0.505, rotation=<rot 0 -90 0>}]
 dyn_hair_strap_a_base_r [{shapeType=Capsule, length=0.505, radius=0.505, rotation=<rot 0 -90 0>}]
 dyn_hair_strap_a_l [{shapeType=Capsule, length=4, radius=1.5, center=<3.4661 0 0>, rotation=<rot 90 0 0>}]
 dyn_hair_strap_a_r [{shapeType=Capsule, length=2.25, radius=1.5, center=<3.202 0 0>, rotation=<rot 90 0 0>}]
 dyn_hair_strap_b_l [{shapeType=Capsule, length=4, radius=1.5, center=<3.4661 0 0>, rotation=<rot 90 0 0>}]
 dyn_hair_strap_b_r [{shapeType=Capsule, length=3.5, radius=1.5, center=<3.4397 0 0>, rotation=<rot 90 0 0>}]
 dyn_hair_strap_c_l [{shapeType=Capsule, length=4, radius=1.5, center=<3.4661 0 0>, rotation=<rot 90 0 0>}]
 dyn_hair_strap_c_r [{shapeType=Capsule, length=3, radius=1.5, center=<3.086 0 0>, rotation=<rot 90 0 0>}]
 dyn_hair_strap_d_l [{shapeType=Capsule, length=4, radius=1.5, center=<3.4661 0 0>, rotation=<rot 90 0 0>}]
 dyn_hair_strap_d_r [{shapeType=Capsule, length=3.5, radius=1.5, center=<3.1877 0 0>, rotation=<rot 90 0 0>}]
 dyn_hair_strap_e_l [{shapeType=Capsule, length=5.75, radius=1.5, center=<4.5903 0 0>, rotation=<rot 90 0 0>}]
 dyn_hair_strap_e_r [{shapeType=Capsule, length=5.5, radius=1.5, center=<4.7286 0 0>, rotation=<rot 90 0 0>}]
 head [{shapeType=Capsule, length=9, radius=9, center=<4.2261 2.8443 0>, rotation=<rot 90 0 0>}]
 spine_05 [{shapeType=Capsule, length=33, radius=16.75, center=<2.5321 2.1936 -0.0001>}]
}
constraints [
 {
  bone1Name=dyn_hair_strap_a_base_l
  bone2Name=head
  swing1LimitDegrees=45
  swing1Motion=Free
  swing2LimitDegrees=45
  swing2Motion=Free
  twistLimitDegrees=45
  twistMotion=Free
 }
 {
  bone1Name=dyn_hair_strap_a_base_r
  bone2Name=head
  swing1LimitDegrees=45
  swing1Motion=Free
  swing2LimitDegrees=45
  swing2Motion=Free
  twistLimitDegrees=45
  twistMotion=Free
 }
 {bone1Name=dyn_hair_strap_a_l, bone2Name=dyn_hair_strap_a_base_l, swing1LimitDegrees=30, swing2LimitDegrees=30, twistLimitDegrees=5}
 {bone1Name=dyn_hair_strap_b_l, bone2Name=dyn_hair_strap_a_l, swing1LimitDegrees=30, swing2LimitDegrees=30, twistLimitDegrees=5}
 {bone1Name=dyn_hair_strap_c_l, bone2Name=dyn_hair_strap_b_l, swing1LimitDegrees=30, swing2LimitDegrees=30, twistLimitDegrees=5}
 {bone1Name=dyn_hair_strap_d_l, bone2Name=dyn_hair_strap_c_l, swing1LimitDegrees=30, swing2LimitDegrees=30, twistLimitDegrees=5}
 {bone1Name=dyn_hair_strap_e_l, bone2Name=dyn_hair_strap_d_l, swing1LimitDegrees=30, swing2LimitDegrees=30, twistLimitDegrees=5}
 {bone1Name=dyn_hair_strap_a_r, bone2Name=dyn_hair_strap_a_base_r, swing1LimitDegrees=30, swing2LimitDegrees=30, twistLimitDegrees=5}
 {bone1Name=dyn_hair_strap_b_r, bone2Name=dyn_hair_strap_a_r, swing1LimitDegrees=30, swing2LimitDegrees=30, twistLimitDegrees=5}
 {bone1Name=dyn_hair_strap_c_r, bone2Name=dyn_hair_strap_b_r, swing1LimitDegrees=30, swing2LimitDegrees=30, twistLimitDegrees=5}
 {bone1Name=dyn_hair_strap_d_r, bone2Name=dyn_hair_strap_c_r, swing1LimitDegrees=30, swing2LimitDegrees=30, twistLimitDegrees=5}
 {bone1Name=dyn_hair_strap_e_r, bone2Name=dyn_hair_strap_d_r, swing1LimitDegrees=30, swing2LimitDegrees=30, twistLimitDegrees=5}
 {
  bone1Name=head
  bone2Name=spine_05
  swing1LimitDegrees=45
  swing1Motion=Free
  swing2LimitDegrees=45
  swing2Motion=Free
  twistLimitDegrees=45
  twistMotion=Free
 }
]
